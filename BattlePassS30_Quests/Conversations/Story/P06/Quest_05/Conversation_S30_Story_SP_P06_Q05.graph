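# Conversation database for Season 30 story quest P06 Q05, giving an NPC a single main-quest line offered as "Talk to Hope".
# An entry point leads to one terminal speech node, whose choice is tagged to auto-select when valid.
# The line requires the NPC to carry the VoidRedemption conversation entry tag and the quest to be active.
# An inverted objective check limits it to before the quest's obj0 objective is completed.
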ConversationDatabase /BattlePassS30_Quests/Conversations/Story/P06/Quest_05/Conversation_S30_Story_SP_P06_Q05
entry_points [{tag=Conversation.Entry.Season30.Story.P06.Q05.VR, destinations [#1]}]
nodes [
 ConversationEntryPointNode_0 {
  id=#1
  class=ConversationEntryPointNode
  outputs [#2]
  properties {entryTag {tagName=Conversation.Entry.Season30.Story.P06.Q05.VR}}
 }
 Speech_C_5 {
  id=#2
  class=Speech_C
  outputs []
  properties {
   softTaskIcon=&T_UI_NPC_MainQuest
   generalConfig {defaultMessage="Well, I'm still waiting for that water-witch to call me back-- Whoa!"}
  }
  sub_nodes [
   ConversationChoiceNode_0 {
    class=ConversationChoiceNode
    properties {
     defaultChoiceDisplayText=Talk to Hope
     choiceTags {gameplayTags [
       {tagName=Conversation.Choice.AutoSelectIfValid}
       {tagName=Conversation.Choice.Category.MainQuestline}
       {tagName=Athena.Quests.S30.Conversation.Story.P06.05.VR}
      ]}
    }
   }
   FortConversationRequirement_ControllerRequirement_0 {
    class=FortConversationRequirement_ControllerRequirement
    properties {
     nodeName=Require: NPC Has Conversation Tag
     participantId {tagName=Conversation.Participant.NPC}
     requirement {class=FortControllerRequirement_NPCHasConversationEntryTag, properties {queryTag {tagName=Conversation.Entry.VoidRedemption}}}
    }
   }
   FortConversationRequirement_ControllerRequirement_1 {
    class=FortConversationRequirement_ControllerRequirement
    properties {
     nodeName=Require: Is Quest Active
     requirement {
      class=FortControllerRequirement_IsQuestActive
      properties {questDefinition=/BattlePassS30_Quests/Items/QuestItems/Story/P06/Quest_S30_Story_P06_Q05}
     }
    }
   }
   FortConversationRequirement_ControllerRequirement_2 {
    class=FortConversationRequirement_ControllerRequirement
    properties {
     nodeName=Require: Has Completed Quest Objective
     requirement {
      class=FortControllerRequirement_HasCompletedQuestObjective
      properties {
       bInverseCondition=true
       objectiveBackendName=Quest_S30_Story_P06_Q05_obj0
       questDefinition=&Quest_S30_Story_P06_Q05
      }
     }
    }
   }
  ]
 }
]
